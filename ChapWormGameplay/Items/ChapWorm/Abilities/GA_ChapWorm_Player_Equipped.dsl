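; Gameplay ability for the ChapWorm item's equipped state on a player, adding reticle aim and custom sprint support while active.
; On activation, authority adds and caches a reticle aim component on the Fort player pawn if none exists.
; It also registers a configured sprint MME class with the controller's tactical sprint component, without an authority check.
; Ending destroys the cached component on authority and unregisters the sprint class before calling the parent.
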(event Ability|EventActivateAbility
  ;[0.01964 0.125 0.03717] ActivateAbility: Apply the reticle aim pawn component and sprint extension class to the avatar player.
  (|CallParentFunction)
  (CallFunction|AddReticleAimPawnComponent)
  (CallFunction|AddSprintExtensionClass))

(event Ability|EventOnEndAbility
  (params (bWasCancelled :type Boolean))
  ;[0.1615 0.1143 0.01825] EndPlay: Remove the added pawn component and sprint extension class from the avatar pawn.
  (Utilities|FlowControl|Sequence
    (:then_0
      (CallFunction|RemoveReticleAimPawnComponent)
      (CallFunction|RemoveSprintExtensionClass))
    (:then_1
      (|CallParentFunction bWasCancelled))))

(fn AddReticleAimPawnComponent
  (Ability|HasAuthority
    (:True
      (bind _outfortplayerpawn (CallFunction|GetAvatarasValidFortPlayerPawn)
        (:ValidFortPlayerPawn
          (Utilities|IsValid (Actor|GetComponentbyClass _outfortplayerpawn "/ChapWormGameplay/Items/ChapWorm/Abilities/B_ChapWormReticleAimPawnComponent.B_ChapWormReticleAimPawnComponent_C")
            (:Is%20Not%20Valid
              (bind _returnvalue (Game|AddComponentbyClass _outfortplayerpawn "/ChapWormGameplay/Items/ChapWorm/Abilities/B_ChapWormReticleAimPawnComponent.B_ChapWormReticleAimPawnComponent_C"))
              (Variables|Transient|SetReticleAimPawnComponent _returnvalue))))))))

(fn AddSprintExtensionClass
  (bind _returnvalue (Ability|GetPlayerControllerfromActorInfo))
  (bind _returnvalue_2 (Actor|GetComponentbyClass _returnvalue "/TacticalSprintGame/Gameplay/TacticalSprintControllerComponent.TacticalSprintControllerComponent_C"))
  (Utilities|IsValid _returnvalue
    (:Is%20Valid
      (Utilities|IsValid _returnvalue_2
        (:Is%20Valid
          (Sprint|AddCustomSprintMMEClass _returnvalue_2 (Variables|Settings|GetSprintMMEClass)))))))

(fn RemoveReticleAimPawnComponent
  (Ability|HasAuthority
    (:True
      (bind _reticleaimpawncomponent (Variables|Transient|GetReticleAimPawnComponent)
        (:then
          (Class|BAutoAimWeaponPawnComponentBase|DestroyComponent _reticleaimpawncomponent))
        (:else)))))

(fn RemoveSprintExtensionClass
  (bind _returnvalue (Ability|GetPlayerControllerfromActorInfo))
  (bind _returnvalue_2 (Actor|GetComponentbyClass _returnvalue "/TacticalSprintGame/Gameplay/TacticalSprintControllerComponent.TacticalSprintControllerComponent_C"))
  (Utilities|IsValid _returnvalue
    (:Is%20Valid
      (Utilities|IsValid _returnvalue_2
        (:Is%20Valid
          (Sprint|RemoveCustomSprintMMEClass _returnvalue_2 (Variables|Settings|GetSprintMMEClass)))))))
